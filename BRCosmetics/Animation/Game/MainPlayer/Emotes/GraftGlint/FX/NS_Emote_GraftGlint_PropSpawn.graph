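# Niagara system for the GraftGlint emote's prop spawn, throwing a one-shot burst of green leaf sprites.
# Active CPU and GPU emitters share box spawning, directional velocity, gravity, aerodynamic drag and spin that stops on collision; Leaves002 and toon smoke are disabled.
# Leaf color follows time of day and darkens indoors, with a fixed green in front-end preview.
# Delayed collision is limited by quality level and off in preview.
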NiagaraSystem /BRCosmetics/Animation/Game/MainPlayer/Emotes/GraftGlint/FX/NS_Emote_GraftGlint_PropSpawn
dependencies {
 usedDataInterfaces [
  &NiagaraDataInterfaceSpriteRendererInfo
  &NiagaraDataInterfaceArrayFloat
  &NiagaraDataInterfaceCurve
  &NiagaraDataInterfacePlatformSet
 ]
 usedDynamicInputs [&Sine]
 usedModules [
  &SystemState
  &self:SystemSpawnScript.SystemScriptSource.SystemScriptGraph.NiagaraNodeAssignment_0.SetVariables_#1
 ]
}
emitters {
 Leaves002 {
  bEnabled=false
  emitterUpdateScript [
   &EmitterState(Life Cycle Mode=::NewEnumerator1, Loop Behavior=::NewEnumerator1, Loop Duration=1, UseLoopDelay=false)
   &SpawnBurst_Instantaneous(Spawn Count=50, Use Spawn Probability=false, UseLoopCountLimit=false)
  ]
  particleSpawnScript [
   &PartitionParticles()
   &InitializeParticle(
    Color=&FNTimeOfDayColor(
     Night Color=<rgba 0.0555 0.1464 0.0656 1>
     Morning Color=<rgba 0.0432 0.2622 0.0676 1>
     Day Color=<rgba 0.0599 0.3791 0.1072 1>
     Evening Color=<rgba 0.0432 0.2622 0.0676 1>
    )
    Lifetime Max=1.5
    Lifetime Min=0.5
    Lifetime Mode=::NewEnumerator1
    Sprite Rotation Mode=::NewEnumerator2
    Sprite Size Max=<16 16>
    Sprite Size Min=<4 4>
    Sprite Size Mode=::NewEnumerator2
    UsePositionOffset=false
    Write Material Random=true
   )
   &SubUVAnimation(
    "Sprite Renderer "=DataInterface() {propertyValues {SpriteRenderer=&self:Leaves002_0.NiagaraSpriteRendererProperties_0}}
    SubUV Animation Mode:ENiagara_SubUVLookupModeV2=::NewEnumerator2
    UseEndFrame=false
    UseStartFrame=false
   )
   &ShapeLocation(
    MaskSpawnGroup=false
    Non Uniform Scale=<1.5 0.2 1>
    Offset Mode=::NewEnumerator0
    Override Seed=false
    Sphere Radius=5
    Sphere Surface Distribution=1
    UseOwnerScale=true
   )
   &AddVelocity(
    Module.Use Falloff Distance=false
    Origin Offset=<0 0 -10>
    Override Seed=false
    Velocity Mode=::NewEnumerator1
    Velocity Speed=&RandomRangeFloat(Minimum=500, Maximum=300)
   ) {module=AddVelocity001}
   &CameraOffset(Camera Offset Amount=10)
   /SetVariables(Particles.Velocity=&Multiply_Vector(A=$Particles.Velocity, B=&TransformVector(Vector=<3 0.2 0.6>, Destination Space=::Local))) {isSetParameters=true}
  ]
  particleUpdateScript [
   &ParticleState()
   &CurlNoiseForce(
    Bias Noise=false
    Noise Frequency=20
    Noise Strength=&MakeCustomFloatFromBool(Bool=$Particles.HasCollided, True Float=50, False Float=500)
    Pan Noise=false
   )
   &SpriteRotationRate(Rotation Rate=&MakeCustomFloatFromBool(
     Bool=$Particles.HasCollided
     True Float=0
     False Float=&SelectFloatFromArray(
      Array Sampling Mode=::NewEnumerator0
      Direct Array Index=?
      Float Selection Array=DataInterface() {propertyValues {FloatData [-2000,-2000,2000,-2000,2000,-2000]}}
      Random Seed=0
     )
    ))
   &GravityForce() {module=GravityForce001}
   &AerodynamicDrag(
    Aerodynamic Drag:NiagaraFloat=&Multiply_Float(
     A=&FloatFromCurve(
      FloatCurve=DataInterface() {propertyValues {Curve {
         defaultValue=1
         keys [
          {interpMode=RCIM_Cubic, tangentMode=RCTM_User, time=0.0051, value=0}
          {arriveTangent=-0.6931, interpMode=RCIM_Cubic, leaveTangent=-0.6931, tangentMode=RCTM_User, time=0.2699, value=1}
         ]
        }}}
      Scale Curve=2
     )
     B=&RandomRangeFloat(Minimum=1, Maximum=1)
    )
    Drag Shape Curve=DataInterface() {propertyValues {Curve {keys [
        {interpMode=RCIM_Cubic, tangentMode=RCTM_User, time=0, value=0.1}
        {arriveTangent=2.0799, interpMode=RCIM_Cubic, leaveTangent=2.0799, tangentMode=RCTM_User, time=1, value=1}
       ]}}}
    Lift Contribution=2
    Lift Shape Curve=DataInterface() {propertyValues {Curve {keys [
        {interpMode=RCIM_Cubic, tangentMode=RCTM_User, time=0, value=0}
        {interpMode=RCIM_Cubic, tangentMode=RCTM_User, time=0.5, value=1}
        {tangentMode=RCTM_User, time=1, value=0}
       ]}}}
    Pivot Offset Maximum:Vector3f=<5 0.5 1>
    Pivot Offset Minimum:Vector3f=<-5 -0.5 -1>
    Rotation Strength:NiagaraFloat=&RandomRangeFloat(Minimum=1.8, Maximum=2)
    Use Lift Smoothing=true
    Use Linear Drag=true
    Use Rotational Drag=true
   ) {module=AerodynamicDrag001}
   &Collision(
    Age Colliding Particles=false
    Collision Enabled=&Boolean_AndOperation(
     Bool1=&Boolean_AndOperation(
      Bool1=&IsPlatformSetActive(New Platform Set=DataInterface() {propertyValues {Platforms {qualityLevelMask=-8}}})
      Bool2=&Boolean_NotOperation(Bool=$User.bIsFrontEndPreview)
     )
     Bool2=&SetBoolByFloatComparison(
      A=&FloatFromCurve(FloatCurve=DataInterface() {propertyValues {Curve {
          defaultValue=1
          keys [{interpMode=RCIM_Cubic, tangentMode=RCTM_User, time=0.2, value=0}, {interpMode=RCIM_Cubic, tangentMode=RCTM_User, time=0.3, value=1}]
         }}})
      B=0.5
     )
    )
    Control Roll On Collision=true
    Enable Rest State=false
    Friction=0.1
    KillOnCollision=false
    Module.EnableMaxCollisionCount=false
    Module.Randomize Collision Normal Vector=false
    Particle Radius Scale=5
    Radius Calculation Type=::NewEnumerator1
    Restitution=0.4
    ScaleRateOfRotation=true
   )
   &Collision(
    Age Colliding Particles=false
    Analytical Collision Plane Space:ENiagaraCoordinateSpace=::Local
    CPU Collision Type=::NewEnumerator3
    Collision Enabled=&Boolean_AndOperation(
     Bool1=&Boolean_AndOperation(
      Bool1=&IsPlatformSetActive(New Platform Set=DataInterface() {propertyValues {Platforms {qualityLevelMask=-8}}})
      Bool2=&Boolean_NotOperation(Bool=$User.bIsFrontEndPreview)
     )
     Bool2=&SetBoolByFloatComparison(
      A=&FloatFromCurve(FloatCurve=DataInterface() {propertyValues {Curve {
          defaultValue=1
          keys [{interpMode=RCIM_Cubic, tangentMode=RCTM_User, time=0.2, value=0}, {interpMode=RCIM_Cubic, tangentMode=RCTM_User, time=0.3, value=1}]
         }}})
      B=0.5
     )
    )
    Control Roll On Collision=true
    Enable Rest State=false
    Friction=0.1
    KillOnCollision=false
    Module.EnableMaxCollisionCount=false
    Module.Randomize Collision Normal Vector=false
    Particle Radius Scale=5
    Perform 2 Analytical Collision Queries=false
    Radius Calculation Type=::NewEnumerator1
    Restitution=0.8
    ScaleRateOfRotation=false
   ) {module=Collision002}
   &SolveForcesAndVelocity(Clamp Velocity=false, Limit Acceleration=false)
   &AlignSpriteToMeshOrientation(SetAlignment=true, SetFacing=true)
   &ScaleSpriteSizeBySpeed(Max Scale Factor=<10 10>)
   &ScaleSpriteSize(Uniform Curve Sprite Scale=DataInterface() {propertyValues {Curve {
       defaultValue=0
       keys [
        {interpMode=RCIM_Cubic, tangentMode=RCTM_User, time=0, value=0}
        {arriveTangent=0.8509, interpMode=RCIM_Cubic, leaveTangent=0.8509, time=0.1789, value=0.7227}
        {interpMode=RCIM_Cubic, tangentMode=RCTM_User, time=0.4558, value=0.3878}
        {arriveTangent=-3.1468, interpMode=RCIM_Cubic, leaveTangent=-3.1468, tangentMode=RCTM_User, time=1, value=0}
       ]
      }}})
  ]
  rendererClasses [&NiagaraSpriteRendererProperties]
  renderers [{rendererClass=&NiagaraSpriteRendererProperties}]
 }
 LeavesCPU {
  emitterUpdateScript [
   &EmitterState(Life Cycle Mode=::NewEnumerator1, Loop Behavior=::NewEnumerator1, Loop Duration=1, UseLoopDelay=false)
   &SpawnBurst_Instantaneous(Spawn Count=50, Use Spawn Probability=false, UseLoopCountLimit=false)
  ]
  particleSpawnScript [
   &InitializeParticle(
    Color=&MakeCustomLinearColorFromBool(
     Bool=$User.bIsFrontEndPreview
     True Color=<rgba 0.0594 0.2448 0.0616 1>
     False Color=&LerpLinearColors(
      StartColor=&FNTimeOfDayColor(
       Night Color=<rgba 0.0555 0.1464 0.0656 1>
       Morning Color=<rgba 0.0432 0.2622 0.0676 1>
       Day Color=<rgba 0.0595 0.3813 0.107 1>
       Evening Color=<rgba 0.0432 0.2622 0.0676 1>
      )
      EndColor=&FNTimeOfDayColor(
       Night Color=<rgba 0.001 0.0087 0.0067 1>
       Morning Color=<rgba 0.0022 0.0133 0.0034 1>
       Day Color=<rgba 0.0028 0.0179 0.0051 1>
       Evening Color=<rgba 0.003 0.0179 0.0046 1>
      )
      LerpFactor=$NPC.FortniteNPC.IsPlayerIndoors
     )
    )
    Lifetime Max=1.3
    Lifetime Min=0.5
    Lifetime Mode=::NewEnumerator1
    Sprite Rotation Mode=::NewEnumerator2
    Sprite Size Max=<16 16>
    Sprite Size Min=<4 4>
    Sprite Size Mode=::NewEnumerator2
    UsePositionOffset=false
    Write Material Random=true
   )
   &SubUVAnimation(
    "Sprite Renderer "=DataInterface() {propertyValues {SpriteRenderer=&self:LeavesCPU_0.NiagaraSpriteRendererProperties_0}}
    SubUV Animation Mode:ENiagara_SubUVLookupModeV2=::NewEnumerator2
    UseEndFrame=false
    UseStartFrame=false
   )
   &ShapeLocation(
    Box Size=<40 10 50>
    MaskSpawnGroup=false
    Non Uniform Scale=<0.5 0.2 0.2>
    Offset Mode=::NewEnumerator0
    Override Seed=false
    Shape Primitive=::NewEnumerator2
    Surface Only=false
    UseOwnerScale=true
   ) {module=ShapeLocation001}
   &CameraOffset(Camera Offset Amount=10)
   &AddVelocity(
    Module.Use Falloff Distance=false
    Override Seed=false
    Velocity Mode=::NewEnumerator1
    Velocity Speed=&RandomRangeFloat(Minimum=500, Maximum=300)
   )
   /SetVariables(Particles.Velocity=&Multiply_Vector(A=$Particles.Velocity, B=&TransformVector(Vector=<3 0.2 1>, Destination Space=::Local))) {isSetParameters=true, module=SetVariables_#2}
  ]
  particleUpdateScript [
   &ParticleState()
   &SpriteRotationRate(Rotation Rate=&MakeCustomFloatFromBool(
     Bool=$Particles.HasCollided
     True Float=0
     False Float=&SelectFloatFromArray(
      Array Sampling Mode=::NewEnumerator0
      Direct Array Index=?
      Float Selection Array=DataInterface() {propertyValues {FloatData [-2000,-2000,2000,-2000,2000,-2000]}}
      Random Seed=0
     )
    ))
   &GravityForce() {module=GravityForce001}
   &AerodynamicDrag(
    Aerodynamic Drag:NiagaraFloat=&Multiply_Float(
     A=&FloatFromCurve(
      FloatCurve=DataInterface() {propertyValues {Curve {
         defaultValue=1
         keys [
          {interpMode=RCIM_Cubic, tangentMode=RCTM_User, time=0.0051, value=0}
          {arriveTangent=-0.6931, interpMode=RCIM_Cubic, leaveTangent=-0.6931, tangentMode=RCTM_User, time=0.2699, value=1}
         ]
        }}}
      Scale Curve=2
     )
     B=&RandomRangeFloat(Minimum=1, Maximum=1)
    )
    Drag Shape Curve=DataInterface() {propertyValues {Curve {keys [
        {interpMode=RCIM_Cubic, tangentMode=RCTM_User, time=0, value=0.1}
        {arriveTangent=2.0799, interpMode=RCIM_Cubic, leaveTangent=2.0799, tangentMode=RCTM_User, time=1, value=1}
       ]}}}
    Lift Contribution=2
    Lift Shape Curve=DataInterface() {propertyValues {Curve {keys [
        {interpMode=RCIM_Cubic, tangentMode=RCTM_User, time=0, value=0}
        {interpMode=RCIM_Cubic, tangentMode=RCTM_User, time=0.5, value=1}
        {tangentMode=RCTM_User, time=1, value=0}
       ]}}}
    Pivot Offset Maximum:Vector3f=<5 0.5 1>
    Pivot Offset Minimum:Vector3f=<-5 -0.5 -1>
    Rotation Strength:NiagaraFloat=&RandomRangeFloat(Minimum=1.8, Maximum=2)
    Use Lift Smoothing=true
    Use Linear Drag=true
    Use Rotational Drag=true
   ) {module=AerodynamicDrag001}
   &Collision(
    Age Colliding Particles=false
    Collision Enabled=&Boolean_AndOperation(
     Bool1=&Boolean_AndOperation(
      Bool1=&IsPlatformSetActive(New Platform Set=DataInterface() {propertyValues {Platforms {qualityLevelMask=-8}}})
      Bool2=&Boolean_NotOperation(Bool=$User.bIsFrontEndPreview)
     )
     Bool2=&SetBoolByFloatComparison(
      A=&FloatFromCurve(FloatCurve=DataInterface() {propertyValues {Curve {
          defaultValue=1
          keys [{interpMode=RCIM_Cubic, tangentMode=RCTM_User, time=0.2, value=0}, {interpMode=RCIM_Cubic, tangentMode=RCTM_User, time=0.3, value=1}]
         }}})
      B=0.5
     )
    )
    Control Roll On Collision=true
    Enable Rest State=false
    Friction=0.1
    KillOnCollision=false
    Module.EnableMaxCollisionCount=false
    Module.Randomize Collision Normal Vector=false
    Particle Radius Scale=5
    Radius Calculation Type=::NewEnumerator1
    Restitution=0.4
    ScaleRateOfRotation=true
   )
   &Collision(
    Age Colliding Particles=false
    Analytical Collision Plane Space:ENiagaraCoordinateSpace=::Local
    CPU Collision Type=::NewEnumerator3
    Collision Enabled=&Boolean_AndOperation(
     Bool1=&Boolean_AndOperation(
      Bool1=&IsPlatformSetActive(New Platform Set=DataInterface() {propertyValues {Platforms {qualityLevelMask=-8}}})
      Bool2=&Boolean_NotOperation(Bool=$User.bIsFrontEndPreview)
     )
     Bool2=&SetBoolByFloatComparison(
      A=&FloatFromCurve(FloatCurve=DataInterface() {propertyValues {Curve {
          defaultValue=1
          keys [{interpMode=RCIM_Cubic, tangentMode=RCTM_User, time=0.2, value=0}, {interpMode=RCIM_Cubic, tangentMode=RCTM_User, time=0.3, value=1}]
         }}})
      B=0.5
     )
    )
    Control Roll On Collision=true
    Enable Rest State=false
    Friction=0.1
    KillOnCollision=false
    Module.EnableMaxCollisionCount=false
    Module.Randomize Collision Normal Vector=false
    Particle Radius Scale=5
    Perform 2 Analytical Collision Queries=false
    Radius Calculation Type=::NewEnumerator1
    Restitution=0.8
    ScaleRateOfRotation=false
   ) {module=Collision002}
   &SolveForcesAndVelocity(Clamp Velocity=false, Limit Acceleration=false)
   &AlignSpriteToMeshOrientation(SetAlignment=true, SetFacing=true)
   &ScaleSpriteSizeBySpeed(Max Scale Factor=<2 4>)
   &ScaleSpriteSize(Uniform Curve Sprite Scale=DataInterface() {propertyValues {Curve {
       defaultValue=0
       keys [
        {interpMode=RCIM_Cubic, tangentMode=RCTM_User, time=0, value=0}
        {arriveTangent=0.8509, interpMode=RCIM_Cubic, leaveTangent=0.8509, time=0.1789, value=0.7227}
        {interpMode=RCIM_Cubic, tangentMode=RCTM_User, time=0.4558, value=0.3878}
        {arriveTangent=-3.1468, interpMode=RCIM_Cubic, leaveTangent=-3.1468, tangentMode=RCTM_User, time=1, value=0}
       ]
      }}})
  ]
  rendererClasses [&NiagaraSpriteRendererProperties]
  renderers [{rendererClass=&NiagaraSpriteRendererProperties}]
 }
 LeavesGPU002 {
  simTarget=GPUComputeSim
  emitterUpdateScript [
   &EmitterState(Life Cycle Mode=::NewEnumerator1, Loop Behavior=::NewEnumerator1, Loop Duration=1, UseLoopDelay=false)
   &SpawnBurst_Instantaneous(Spawn Count=300, Use Spawn Probability=false, UseLoopCountLimit=false)
  ]
  particleSpawnScript [
   &InitializeParticle(
    Color=&MakeCustomLinearColorFromBool(
     Bool=$User.bIsFrontEndPreview
     True Color=<rgba 0.0594 0.2448 0.0616 1>
     False Color=&LerpLinearColors(
      StartColor=&FNTimeOfDayColor(
       Night Color=<rgba 0.0555 0.1464 0.0656 1>
       Morning Color=<rgba 0.0432 0.2622 0.0676 1>
       Day Color=<rgba 0.0595 0.3813 0.107 1>
       Evening Color=<rgba 0.0432 0.2622 0.0676 1>
      )
      EndColor=&FNTimeOfDayColor(
       Night Color=<rgba 0.001 0.0087 0.0067 1>
       Morning Color=<rgba 0.0022 0.0133 0.0034 1>
       Day Color=<rgba 0.0028 0.0179 0.0051 1>
       Evening Color=<rgba 0.003 0.0179 0.0046 1>
      )
      LerpFactor=$NPC.FortniteNPC.IsPlayerIndoors
     )
    )
    Lifetime Max=1.3
    Lifetime Min=0.5
    Lifetime Mode=::NewEnumerator1
    Sprite Rotation Mode=::NewEnumerator2
    Sprite Size Max=<16 16>
    Sprite Size Min=<4 4>
    Sprite Size Mode=::NewEnumerator2
    UsePositionOffset=false
    Write Material Random=true
   )
   &SubUVAnimation(
    "Sprite Renderer "=DataInterface() {propertyValues {SpriteRenderer=&self:LeavesGPU002_0.NiagaraSpriteRendererProperties_0}}
    SubUV Animation Mode:ENiagara_SubUVLookupModeV2=::NewEnumerator2
    UseEndFrame=false
    UseStartFrame=false
   )
   &ShapeLocation(
    Box Size=<40 10 50>
    MaskSpawnGroup=false
    Non Uniform Scale=<0.5 0.2 0.2>
    Offset Mode=::NewEnumerator0
    Override Seed=false
    Shape Primitive=::NewEnumerator2
    Surface Only=false
    UseOwnerScale=true
   ) {module=ShapeLocation001}
   &CameraOffset(Camera Offset Amount=10)
   &AddVelocity(
    Module.Use Falloff Distance=false
    Override Seed=false
    Velocity Mode=::NewEnumerator1
    Velocity Speed=&RandomRangeFloat(Minimum=500, Maximum=300)
   )
   /SetVariables(Particles.Velocity=&Multiply_Vector(A=$Particles.Velocity, B=&TransformVector(Vector=<3 0.2 1>, Destination Space=::Local))) {isSetParameters=true, module=SetVariables_#2}
  ]
  particleUpdateScript [
   &ParticleState()
   &SpriteRotationRate(Rotation Rate=&MakeCustomFloatFromBool(
     Bool=$Particles.HasCollided
     True Float=0
     False Float=&SelectFloatFromArray(
      Array Sampling Mode=::NewEnumerator0
      Direct Array Index=?
      Float Selection Array=DataInterface() {propertyValues {FloatData [-2000,-2000,2000,-2000,2000,-2000]}}
      Random Seed=0
     )
    ))
   &GravityForce() {module=GravityForce001}
   &AerodynamicDrag(
    Aerodynamic Drag:NiagaraFloat=&Multiply_Float(
     A=&FloatFromCurve(
      FloatCurve=DataInterface() {propertyValues {Curve {
         defaultValue=1
         keys [
          {interpMode=RCIM_Cubic, tangentMode=RCTM_User, time=0.0051, value=0}
          {arriveTangent=-0.6931, interpMode=RCIM_Cubic, leaveTangent=-0.6931, tangentMode=RCTM_User, time=0.2699, value=1}
         ]
        }}}
      Scale Curve=2
     )
     B=&RandomRangeFloat(Minimum=1, Maximum=1)
    )
    Drag Shape Curve=DataInterface() {propertyValues {Curve {keys [
        {interpMode=RCIM_Cubic, tangentMode=RCTM_User, time=0, value=0.1}
        {arriveTangent=2.0799, interpMode=RCIM_Cubic, leaveTangent=2.0799, tangentMode=RCTM_User, time=1, value=1}
       ]}}}
    Lift Contribution=2
    Lift Shape Curve=DataInterface() {propertyValues {Curve {keys [
        {interpMode=RCIM_Cubic, tangentMode=RCTM_User, time=0, value=0}
        {interpMode=RCIM_Cubic, tangentMode=RCTM_User, time=0.5, value=1}
        {tangentMode=RCTM_User, time=1, value=0}
       ]}}}
    Pivot Offset Maximum:Vector3f=<5 0.5 1>
    Pivot Offset Minimum:Vector3f=<-5 -0.5 -1>
    Rotation Strength:NiagaraFloat=&RandomRangeFloat(Minimum=1.8, Maximum=2)
    Use Lift Smoothing=true
    Use Linear Drag=true
    Use Rotational Drag=true
   ) {module=AerodynamicDrag001}
   &Collision(
    Age Colliding Particles=false
    Collision Enabled=&Boolean_AndOperation(
     Bool1=&Boolean_AndOperation(
      Bool1=&IsPlatformSetActive(New Platform Set=DataInterface() {propertyValues {Platforms {qualityLevelMask=-8}}})
      Bool2=&Boolean_NotOperation(Bool=$User.bIsFrontEndPreview)
     )
     Bool2=&SetBoolByFloatComparison(
      A=&FloatFromCurve(FloatCurve=DataInterface() {propertyValues {Curve {
          defaultValue=1
          keys [{interpMode=RCIM_Cubic, tangentMode=RCTM_User, time=0.2, value=0}, {interpMode=RCIM_Cubic, tangentMode=RCTM_User, time=0.3, value=1}]
         }}})
      B=0.5
     )
    )
    Control Roll On Collision=true
    Enable Rest State=false
    Friction=0.1
    Kill Occluded Particles=false
    Kill Offscreen Particles=true
    KillOnCollision=false
    Module.EnableMaxCollisionCount=false
    Module.Randomize Collision Normal Vector=false
    Particle Radius Scale=5
    Radius Calculation Type=::NewEnumerator1
    Restitution=0.4
    ScaleRateOfRotation=true
    Z Depth Query Type:SceneDepthCollisionQueryMethod=::NewEnumerator2
   )
   &Collision(
    Age Colliding Particles=false
    Collision Enabled=&Boolean_AndOperation(
     Bool1=&Boolean_AndOperation(
      Bool1=&IsPlatformSetActive(New Platform Set=DataInterface() {propertyValues {Platforms {qualityLevelMask=-8}}})
      Bool2=&Boolean_NotOperation(Bool=$User.bIsFrontEndPreview)
     )
     Bool2=&SetBoolByFloatComparison(
      A=&FloatFromCurve(FloatCurve=DataInterface() {propertyValues {Curve {
          defaultValue=1
          keys [{interpMode=RCIM_Cubic, tangentMode=RCTM_User, time=0.2, value=0}, {interpMode=RCIM_Cubic, tangentMode=RCTM_User, time=0.3, value=1}]
         }}})
      B=0.5
     )
    )
    Control Roll On Collision=true
    Enable Rest State=false
    Friction=0.1
    Kill Occluded Particles=false
    Kill Offscreen Particles=true
    KillOnCollision=false
    Module.EnableMaxCollisionCount=false
    Module.Randomize Collision Normal Vector=false
    Particle Radius Scale=5
    Radius Calculation Type=::NewEnumerator1
    Restitution=0.8
    ScaleRateOfRotation=false
    Z Depth Query Type:SceneDepthCollisionQueryMethod=::NewEnumerator2
   ) {module=Collision002}
   &SolveForcesAndVelocity(Clamp Velocity=false, Limit Acceleration=false)
   &AlignSpriteToMeshOrientation(SetAlignment=true, SetFacing=true)
   &ScaleSpriteSizeBySpeed(Max Scale Factor=<2 4>)
   &ScaleSpriteSize(Uniform Curve Sprite Scale=DataInterface() {propertyValues {Curve {
       defaultValue=0
       keys [
        {interpMode=RCIM_Cubic, tangentMode=RCTM_User, time=0, value=0}
        {arriveTangent=0.8509, interpMode=RCIM_Cubic, leaveTangent=0.8509, time=0.1789, value=0.7227}
        {interpMode=RCIM_Cubic, tangentMode=RCTM_User, time=0.4558, value=0.3878}
        {arriveTangent=-3.1468, interpMode=RCIM_Cubic, leaveTangent=-3.1468, tangentMode=RCTM_User, time=1, value=0}
       ]
      }}})
  ]
  rendererClasses [&NiagaraSpriteRendererProperties]
  renderers [{rendererClass=&NiagaraSpriteRendererProperties}]
 }
 ToonSmoke {bEnabled=false, rendererClasses [], renderers []}
 ToonSmoke_LowMedium {bEnabled=false, rendererClasses [], renderers []}
}
summary {
 systemName=NS_Emote_GraftGlint_PropSpawn
 emitters [
  {bEnabled=false, emitterName=ToonSmoke, rendererClasses []}
  {bEnabled=false, emitterName=ToonSmoke_LowMedium, rendererClasses []}
  {bEnabled=false, emitterName=Leaves002, rendererClasses [&NiagaraSpriteRendererProperties]}
  {emitterName=LeavesGPU002, simTarget=GPUComputeSim, rendererClasses [&NiagaraSpriteRendererProperties]}
  {emitterName=LeavesCPU, rendererClasses [&NiagaraSpriteRendererProperties]}
 ]
 userVariables [
  User.bIsFrontEndPreview {type=NiagaraBool}
  User.OwnerLocalVelocity {type=Vector3f, description="", defaultValue {struct=Vector3f, value=<0 0 0>}}
 ]
}
